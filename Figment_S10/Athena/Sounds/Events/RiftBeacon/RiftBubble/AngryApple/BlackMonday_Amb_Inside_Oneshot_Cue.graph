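# Sound cue that plays one-shot ambient sounds for the inside of the BlackMonday (AngryApple) rift bubble.
# A root random node picks evenly between a creature branch and a siren branch, each randomly choosing one equally weighted wave.
# A single-input mixer lowers the creature branch's volume relative to the sirens.
# The cue uses the Fort_Gadgets sound class and overrides attenuation with distance volume falloff turned off.
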SoundCue /Figment_S10/Athena/Sounds/Events/RiftBeacon/RiftBubble/AngryApple/BlackMonday_Amb_Inside_Oneshot_Cue
root=Random_4
nodes [
 {id=Mixer_0, children [Random_2], properties {inputVolume [0.6]}}
 {id=Random_2, children [WavePlayer_18, WavePlayer_19, WavePlayer_20, WavePlayer_21], properties {weights [1,1,1,1]}}
 {
  id=Random_3
  children [WavePlayer_22, WavePlayer_23, WavePlayer_25, WavePlayer_26, WavePlayer_27, WavePlayer_28]
  properties {weights [1,1,1,1,1,1]}
 }
 {id=Random_4, children [Mixer_0,Random_3], properties {weights [1,1]}}
 {id=WavePlayer_18, properties {soundWaveAssetPtr=&/BlackMonday_Creature_04}}
 {id=WavePlayer_19, properties {soundWaveAssetPtr=&/BlackMonday_Creature_02}}
 {id=WavePlayer_20, properties {soundWaveAssetPtr=&/BlackMonday_Creature_03}}
 {id=WavePlayer_21, properties {soundWaveAssetPtr=&/BlackMonday_Creature_01}}
 {id=WavePlayer_22, properties {soundWaveAssetPtr=&/BlackMonday_Siren_01}}
 {id=WavePlayer_23, properties {soundWaveAssetPtr=&/BlackMonday_Siren_02}}
 {id=WavePlayer_25, properties {soundWaveAssetPtr=&/BlackMonday_Siren_04}}
 {id=WavePlayer_26, properties {soundWaveAssetPtr=&/BlackMonday_Siren_05}}
 {id=WavePlayer_27, properties {soundWaveAssetPtr=&/BlackMonday_Siren_06}}
 {id=WavePlayer_28, properties {soundWaveAssetPtr=&/BlackMonday_Siren_07}}
]
properties {
 bOverrideAttenuation=true
 maxDistance=2097152
 soundClassObject=&Fort_Gadgets
 volumeMultiplier=1.1
 attenuationOverrides {
  bAttenuate=false
  bEnableFocusInterpolation=true
  falloffDistance=100000
  focusAzimuth=15
  nonFocusAzimuth=30
  nonFocusVolumeAttenuation=0.2
  stereoSpread=2048
  attenuationShapeExtents {x=0}
 }
}
