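# Sound cue for a blinking-light click in the BlackMonday rift beacon bubble event, judging by its name and path.
# A reused shotgun click wave feeds a Modulator at default settings for slight variation, routed to the gadgets sound class at reduced volume.
# Overridden short-range attenuation adds low-pass filtering, manual reverb send and listener focus that quiets off-focus sources.
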SoundCue /Figment_S10/Athena/Sounds/Events/RiftBeacon/RiftBubble/AngryApple/BlackMonday_LightBlink_Cue
root=Modulator_0
nodes [
 {id=Modulator_0, children [WavePlayer_6]}
 {id=WavePlayer_6, properties {soundWaveAssetPtr=/Game/Sounds/Fort_Weapons/Shotgun/Reload_Elements/weapon_shotgun_2_click}}
]
properties {
 bOverrideAttenuation=true
 maxDistance=768
 soundClassObject=&Fort_Gadgets
 volumeMultiplier=0.7
 attenuationOverrides {
  bAttenuateWithLPF=true
  bEnableFocusInterpolation=true
  bEnableListenerFocus=true
  falloffDistance=768
  focusAzimuth=15
  focusDistanceScale=0.75
  lPFFrequencyAtMax=500
  lPFRadiusMax=0
  lPFRadiusMin=0
  manualReverbSendLevel=50
  nonFocusAzimuth=30
  nonFocusVolumeAttenuation=0.2
  reverbSendMethod=Manual
  reverbWetLevelMin=0
  stereoSpread=256
  attenuationShapeExtents {x=0}
 }
}
