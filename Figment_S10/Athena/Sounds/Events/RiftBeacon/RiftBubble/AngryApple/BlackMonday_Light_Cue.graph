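# Sound cue for the BlackMonday light in the AngryApple rift bubble event, playing a continuous stadium light hum.
# A single looping wave player feeds the output directly, reusing a Fortnite exterior prop sound under the gadgets sound class.
# Overridden attenuation keeps the hum short-range with slight stereo spread, and uses interpolated listener focus to quiet it when outside a narrow facing cone.
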SoundCue /Figment_S10/Athena/Sounds/Events/RiftBeacon/RiftBubble/AngryApple/BlackMonday_Light_Cue
root=WavePlayer_1
nodes [{id=WavePlayer_1, properties {bLooping=true, soundWaveAssetPtr=/Game/Sounds/Fort_Prop_Sounds/Exterior/StadiumLight_Hum_Loop_01}}]
properties {
 bOverrideAttenuation=true
 maxDistance=768
 soundClassObject=&Fort_Gadgets
 attenuationOverrides {
  bEnableFocusInterpolation=true
  bEnableListenerFocus=true
  falloffDistance=768
  focusAzimuth=15
  nonFocusAzimuth=30
  nonFocusVolumeAttenuation=0.2
  stereoSpread=256
  attenuationShapeExtents {x=0}
 }
}
